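# Niagara system for the Sparkles vehicle boost cosmetic, trailing glowing ribbons and twinkling star sprites from boost sockets on the car's skeletal mesh.
# Ribbon_Default and Ribbon_Color are near-twin three-renderer emitters; arrays indexed by ColorVariantID zero one's spawn rate, separating unpainted from painted looks.
# Stars spreads SubUV sprites across ports, growing and fading them.
# All emitters use painted-variant color overrides; ribbons also animate glow and zigzag values over life.
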NiagaraSystem /VehicleCosmetics/Boosts/Boost_Sparkles/FX/NS_Boost_Sparkles
dependencies {
 usedDataInterfaces [
  &NiagaraDataInterfaceCurve
  &NiagaraDataInterfaceSpriteRendererInfo
  &NiagaraDataInterfaceArrayFloat
  &NiagaraDataInterfaceSkeletalMesh
 ]
 usedDynamicInputs []
 usedModules [
  &self:SystemSpawnScript.SystemScriptSource.SystemScriptGraph.NiagaraNodeAssignment_0.SetVariables_#1
  &SystemState
 ]
}
emitters {
 Ribbon_Color {
  emitterUpdateScript [
   &EmitterState(Inactive Response=::NewEnumerator1, Life Cycle Mode=::NewEnumerator1, Loop Duration=1, UseLoopDelay=false)
   &VC_FilteredSocketCount()
   &VC_RateQualityHandler(Epic Spawn Rate=45, Low Spawn Rate=30, Use Custom Rate?=true) {enabled=false}
   &VC_SpawnRibbon_v1(
    Number of Ribbons=$Emitter.PortCount
    SpawnRate=&SelectFloatFromArray(
     Direct Array Index=$User.ColorVariantID
     Float Selection Array=DataInterface() {propertyValues {FloatData [0,60,60,60,60,60,60,60,60,60,60,60,60,60,60]}}
    )
   )
  ]
  particleSpawnScript [
   &VC_PaintedVariantsOverride(
    BurntSienna=<rgba 0.349 0.2265 0.1242 0.2>
    Cobalt=<rgba 0.1024 0.119 1 0.2>
    Crimson=<rgba 0.6233 0.092 0.1064 0.1>
    DefaultColor=<rgba 1 0.5575 0.2587 0.2>
    ForestGreen=<rgba 0 0.4462 0.0225 0.25>
    Grey=<rgba 0.3038 0.3038 0.3038 0.25>
    Lime=<rgba 0.33 0.5 0 0.3>
    Pink=<rgba 0.6927 0.4197 0.5934 0.1>
    Purple=<rgba 0.2453 0.0837 0.4774 0.3>
    R1 BurntSienna:LinearColor=<rgba 0.2865 0.1559 0.0532 1>
    R1 Cobalt:LinearColor=<rgba 0.0573 0.1759 1 1>
    R1 Crimson:LinearColor=<rgba 0.467 0.0501 0 1>
    R1 DefaultColor=<rgba 1.2 0.5861 0 1>
    R1 ForestGreen:LinearColor=<rgba 0 0.4253 0.0005 1>
    R1 Pink:LinearColor=<rgba 0.7309 0.2551 0.5579 1>
    R2 BurntSienna:LinearColor=<rgba 0.5538 0.3044 0.2375 1>
    R2 Cobalt:LinearColor=<rgba 0.4948 0.5041 1 1>
    R2 Crimson:LinearColor=<rgba 1 0 0.0271 1>
    R2 DefaultColor:LinearColor=<rgba 1 0.7588 0.4149 1>
    R2 ForestGreen:LinearColor=<rgba 0 0.8038 0.0406 1>
    R2 Grey:LinearColor=<rgba 1 1 1 0.8>
    R2 Lime:LinearColor=<rgba 0.7794 1 0 1>
    R2 Pink:LinearColor=<rgba 1 0.3281 0.7557 1>
    R2 Purple:LinearColor=<rgba 0.4176 0.0122 1 1>
    R2 Saffron:LinearColor=<rgba 0.9047 1 0 1>
    R2 SkyBlue:LinearColor=<rgba 0 0.6654 1 1>
    Saffron=<rgba 0.3424 0.4 0.08 0.4>
    SkyBlue=<rgba 0.2366 0.5748 0.7448 0.1>
    TitaniumWhite=<rgba 0.6 0.6 0.6 0.1>
    Use R1 Painted Overrides=true
    Use R2 Painted Overrides=true
   )
   &VC_InitializeRibbon_v1(
    Color=$Particles.Color
    Custom Offset=true
    Enable Skeletal Position=true
    Lifetime=0.8
    Local Vector Offset=&AddVectorToPosition(Position=<-30 0 0>, Vector=$User.DTCarOffset)
    Offset Amount=1
    Ribbon Width=0
    Write Material Random=false
   )
   &AddVelocity(Velocity=<1750 0 0>, Velocity Speed Scale=-1) {module=AddVelocity001}
   /VehicleCosmetics/FX/VC_FXParents/Modules/DLR_ProductPreviews(GameModeID:NiagaraInt32=$System.GameModeID, bPreviewVehicleSpeed:NiagaraBool=$System.bPreviewVehicleSpeed)
  ]
  particleUpdateScript [
   &ParticleState()
   &InheritSourceMovement(
    Applied Position Scale=&Multiply_VectorByFloat(Vector=<1 1 1>, Float=1)
    ApplyForce=false
    ApplyPosition=true
    ApplyPositionFalloff=false
   ) {module=InheritSourceMovement001}
   &InheritSourceMovement(
    Applied Position Scale=&VectorFromFloat(Value=&FloatFromCurve(FloatCurve=DataInterface() {propertyValues {Curve {
         defaultValue=1
         keys [
          {arriveTangent=-1.7885, interpMode=RCIM_Cubic, leaveTangent=-1.7885, tangentMode=RCTM_User, time=0.05, value=1}
          {interpMode=RCIM_Cubic, tangentMode=RCTM_User, time=1, value=0}
         ]
        }}}))
    ApplyForce=false
    ApplyPosition=true
    ApplyPositionFalloff=false
   ) {enabled=false}
   &SolveForcesAndVelocity(Clamp Velocity=false, Limit Acceleration=false)
   /SetVariables(
    Particles.Glow:NiagaraFloat=&FloatFromCurve(
     FloatCurve=DataInterface() {propertyValues {Curve {
        defaultValue=1
        keys [
         {arriveTangent=1.3598, interpMode=RCIM_Cubic, leaveTangent=1.3598, tangentMode=RCTM_User, time=0, value=0.15}
         {arriveTangent=-0.0553, interpMode=RCIM_Cubic, leaveTangent=-0.0553, tangentMode=RCTM_User, time=0.4, value=1.1}
         {arriveTangent=-1.3875, interpMode=RCIM_Cubic, leaveTangent=-1.3875, tangentMode=RCTM_User, time=0.7261, value=0.6093}
         {arriveTangent=-0.647, tangentMode=RCTM_User, time=1, value=0.4602}
        ]
       }}}
     Scale Curve=50
    )
    Particles.Stars Disolve=18
    Particles.Zig Zag:NiagaraFloat=&FloatFromCurve(
     FloatCurve=DataInterface() {propertyValues {Curve {
        defaultValue=1
        keys [
         {arriveTangent=1.3598, interpMode=RCIM_Cubic, leaveTangent=1.3598, tangentMode=RCTM_User, time=0, value=0.4}
         {arriveTangent=-0.0553, interpMode=RCIM_Cubic, leaveTangent=-0.0553, tangentMode=RCTM_User, time=0.3, value=0.7}
         {arriveTangent=-0.1551, interpMode=RCIM_Cubic, leaveTangent=-0.1551, tangentMode=RCTM_User, time=0.8, value=0.2}
         {arriveTangent=-0.0957, tangentMode=RCTM_User, time=1, value=0.2}
        ]
       }}}
     Scale Curve=20
    )
   ) {isSetParameters=true, module=SetVariables_#2}
   &DynamicMaterialParameters(
    Index 0 Param 1=1
    Module.Index0Float0Write=true
    Module.Index0Float1Write=true
    Module.Index0Float2Write=true
    Module.Index0Float3Write=true
   )
   &VC_UpdateRibbon_v1(Enable Skeletal Position=true)
  ]
  rendererClasses [&NiagaraRibbonRendererProperties]
  renderers [
   {rendererClass=&NiagaraRibbonRendererProperties}
   {rendererClass=&NiagaraRibbonRendererProperties, rendererIndex=1}
   {rendererClass=&NiagaraRibbonRendererProperties, rendererIndex=2}
  ]
 }
 Ribbon_Default {
  emitterUpdateScript [
   &EmitterState(Inactive Response=::NewEnumerator1, Life Cycle Mode=::NewEnumerator1, Loop Duration=1, UseLoopDelay=false)
   &VC_FilteredSocketCount()
   &VC_RateQualityHandler(Epic Spawn Rate=45, Low Spawn Rate=30, Use Custom Rate?=true) {enabled=false}
   &VC_SpawnRibbon_v1(
    Number of Ribbons=$Emitter.PortCount
    SpawnRate=&SelectFloatFromArray(Direct Array Index=$User.ColorVariantID, Float Selection Array=DataInterface() {propertyValues {FloatData [60,0,0,0,0,0,0,0,0,0,0,0,0,0,0]}})
   )
  ]
  particleSpawnScript [
   &VC_PaintedVariantsOverride(
    BurntSienna=<rgba 0.349 0.2157 0.1799 0.2>
    Cobalt=<rgba 0.2031 0.2179 1 0.25>
    Crimson=<rgba 0.6233 0.092 0.1064 0.1>
    DefaultColor=<rgba 1 0.5575 0.2587 1>
    ForestGreen=<rgba 0 0.5 0.0253 0.25>
    Grey=<rgba 0.3038 0.3038 0.3038 0.25>
    Lime=<rgba 0.277 0.4174 0 0.4841>
    Pink=<rgba 0.6337 0.3795 0.5413 0.1>
    Purple=<rgba 0.2453 0.0837 0.4774 0.5>
    R1 BurntSienna:LinearColor=<rgba 0.0538 0.0423 0.0311 0.002>
    R1 Cobalt:LinearColor=<rgba 0.0692 0.0713 0.1788 0.001>
    R1 Crimson:LinearColor=<rgba 0.1406 0.0378 0.0406 0.001>
    R1 DefaultColor=<rgba 1.2 0.5861 0 1>
    R1 ForestGreen:LinearColor=<rgba 0 0.1372 0.0069 0.001>
    R1 Grey:LinearColor=<rgba 0.1163 0.1163 0.1163 0.001>
    R1 Lime:LinearColor=<rgba 0.0646 0.0747 0 0.005>
    R1 Pink:LinearColor=<rgba 0.1615 0.0866 0.1342 0.001>
    R1 Purple:LinearColor=<rgba 0.0854 0.0228 0.1753 0.001>
    R1 Saffron:LinearColor=<rgba 0.2795 0.2143 0 0.001>
    R1 SkyBlue:LinearColor=<rgba 0 0.0936 0.1406 0.001>
    R1 TitaniumWhite:LinearColor=<rgba 0.1094 0.1094 0.1094 0.001>
    R2 BurntSienna:LinearColor=<rgba 0.5538 0.3044 0.2375 1>
    R2 Cobalt:LinearColor=<rgba 0.4948 0.5041 1 1>
    R2 Crimson:LinearColor=<rgba 1 0 0.0271 1>
    R2 DefaultColor:LinearColor=<rgba 1 0.7588 0.4149 1>
    R2 ForestGreen:LinearColor=<rgba 0 0.8038 0.0406 1>
    R2 Grey:LinearColor=<rgba 1 1 1 0.8>
    R2 Lime:LinearColor=<rgba 0.7794 1 0 1>
    R2 Pink:LinearColor=<rgba 1 0.3281 0.7557 1>
    R2 Purple:LinearColor=<rgba 0.4176 0.0122 1 1>
    R2 Saffron:LinearColor=<rgba 0.9047 1 0 1>
    R2 SkyBlue:LinearColor=<rgba 0 0.6654 1 1>
    Saffron=<rgba 0.3943 0.4809 0 0.5826>
    SkyBlue=<rgba 0.3177 0.7717 1 0.1>
    TitaniumWhite=<rgba 0.75 0.75 0.75 0.1>
    Use R1 Painted Overrides=true
    Use R2 Painted Overrides=true
   )
   &VC_InitializeRibbon_v1(
    Color=$Particles.Color
    Custom Offset=true
    Enable Skeletal Position=true
    Lifetime=0.8
    Local Vector Offset=&AddVectorToPosition(Position=<-30 0 0>, Vector=$User.DTCarOffset)
    Offset Amount=1
    Ribbon Width=0
    Write Material Random=false
   )
   &AddVelocity(Velocity=<1750 0 0>, Velocity Speed Scale=-1) {module=AddVelocity001}
   &DLR_ProductPreviews(GameModeID:NiagaraInt32=$System.GameModeID, bPreviewVehicleSpeed:NiagaraBool=$System.bPreviewVehicleSpeed)
  ]
  particleUpdateScript [
   &ParticleState()
   &InheritSourceMovement(
    Applied Position Scale=&Multiply_VectorByFloat(Vector=<1 1 1>, Float=1)
    ApplyForce=false
    ApplyPosition=true
    ApplyPositionFalloff=false
   ) {module=InheritSourceMovement001}
   &InheritSourceMovement(
    Applied Position Scale=&VectorFromFloat(Value=&FloatFromCurve(FloatCurve=DataInterface() {propertyValues {Curve {
         defaultValue=1
         keys [
          {arriveTangent=-1.7885, interpMode=RCIM_Cubic, leaveTangent=-1.7885, tangentMode=RCTM_User, time=0.05, value=1}
          {interpMode=RCIM_Cubic, tangentMode=RCTM_User, time=1, value=0}
         ]
        }}}))
    ApplyForce=false
    ApplyPosition=true
    ApplyPositionFalloff=false
   ) {enabled=false}
   &SolveForcesAndVelocity(Clamp Velocity=false, Limit Acceleration=false)
   /SetVariables(
    Particles.Glow:NiagaraFloat=&FloatFromCurve(
     FloatCurve=DataInterface() {propertyValues {Curve {
        defaultValue=1
        keys [
         {arriveTangent=1.3598, interpMode=RCIM_Cubic, leaveTangent=1.3598, tangentMode=RCTM_User, time=0, value=0.15}
         {arriveTangent=-0.0553, interpMode=RCIM_Cubic, leaveTangent=-0.0553, tangentMode=RCTM_User, time=0.2989, value=1.1}
         {arriveTangent=-1.3875, interpMode=RCIM_Cubic, leaveTangent=-1.3875, tangentMode=RCTM_User, time=0.7261, value=0.6093}
         {arriveTangent=-0.647, tangentMode=RCTM_User, time=1, value=0.3563}
        ]
       }}}
     Scale Curve=50
    )
    Particles.Stars Disolve=18
    Particles.Zic Zac:NiagaraFloat=&FloatFromCurve(
     FloatCurve=DataInterface() {propertyValues {Curve {
        defaultValue=1
        keys [
         {arriveTangent=1.3598, interpMode=RCIM_Cubic, leaveTangent=1.3598, tangentMode=RCTM_User, time=0, value=0.15}
         {arriveTangent=-0.0553, interpMode=RCIM_Cubic, leaveTangent=-0.0553, tangentMode=RCTM_User, time=0.2989, value=1}
         {arriveTangent=-0.1551, interpMode=RCIM_Cubic, leaveTangent=-0.1551, tangentMode=RCTM_User, time=0.8, value=0.2}
         {arriveTangent=-0.0957, tangentMode=RCTM_User, time=1, value=0.2}
        ]
       }}}
     Scale Curve=25
    )
   ) {isSetParameters=true, module=SetVariables_#2}
   &DynamicMaterialParameters(
    Index 0 Param 1=1
    Module.Index0Float0Write=true
    Module.Index0Float1Write=true
    Module.Index0Float2Write=true
    Module.Index0Float3Write=true
   )
   &VC_UpdateRibbon_v1(Enable Skeletal Position=true)
  ]
  rendererClasses [&NiagaraRibbonRendererProperties]
  renderers [
   {rendererClass=&NiagaraRibbonRendererProperties}
   {rendererClass=&NiagaraRibbonRendererProperties, rendererIndex=1}
   {rendererClass=&NiagaraRibbonRendererProperties, rendererIndex=2}
  ]
 }
 Stars {
  emitterUpdateScript [
   &EmitterState()
   &VC_FilteredSocketCount()
   &SpawnRate(SpawnRate=60, Use Spawn Probability=false)
   /SetVariables("Emitter.RGB ":LinearColor=&MergeHelper_ColorSwitch(0=<rgba 0 0.25 0 0>, 1=<rgba 0 0.8 0 0>, 2=<rgba 0 0.8 0 0>, 3=<rgba 0 0.8 0 0>, Index=$User.GameModeID, Num=::NewEnumerator2)) {isSetParameters=true}
  ]
  particleSpawnScript [
   &VC_PaintedVariantsOverride(
    BurntSienna=<rgba 1 0.5961 0.1823 0.5>
    Cobalt=<rgba 0.467 0.4769 1 1>
    Crimson=<rgba 1 0.0191 0.0457 0.4>
    DefaultColor=<rgba 1 0.6098 0.2899 1>
    ForestGreen=<rgba 0.1173 0.6198 0.1427 1>
    Grey=<rgba 0.5781 0.5781 0.5781 1>
    Lime=<rgba 0.5567 1 0 1>
    Pink=<rgba 1 0.5712 0.8441 1>
    Purple=<rgba 0.4101 0.1021 0.8524 1>
    Saffron=<rgba 0.8421 0.9 0.0909 1>
    SkyBlue=<rgba 0 0.6654 1 1>
   )
   &InitializeParticle(
    Color Mode=::NewEnumerator0
    Lifetime=0.4
    Sprite Rotation Mode=::NewEnumerator2
    Sprite Size Mode=::NewEnumerator3
    Uniform Sprite Size=&RandomRangeFloat(Minimum=150, Maximum=100)
    UsePositionOffset=false
    Write Material Random=true
   )
   &PartitionParticles(Number Of Partitions=$Emitter.PortCount)
   &SkeletalMeshLocation(
    Bone / Socket Index=$Particles.PartitionParticles.Partition
    Bone / Socket Sampling Mode=::NewEnumerator5
    MaskSpawnGroup=false
    Mesh Sampling Type=::NewEnumerator6
    Skeletal Mesh=$System.SkeletalMesh
   )
   &AddVelocity(Velocity=<1750 0 0>, Velocity Speed Scale=-1)
   &DLR_ProductPreviews(GameModeID:NiagaraInt32=$System.GameModeID, bPreviewVehicleSpeed:NiagaraBool=$System.bPreviewVehicleSpeed)
  ]
  particleUpdateScript [
   &ParticleState()
   &ScaleSpriteSize(
    Uniform Curve Scale=0.7
    Uniform Curve Sprite Scale=DataInterface() {propertyValues {Curve {keys [{tangentMode=RCTM_User, time=0, value=0.7}, {tangentMode=RCTM_User, time=1, value=1.8}]}}}
   )
   &InheritSourceMovement(ApplyForce=false, ApplyPosition=true, ApplyPositionFalloff=false)
   &InheritSourceMovement(
    Applied Position Scale=&Multiply_VectorByFloat(
     Vector=<1 1 1>
     Float=&FloatFromCurve(FloatCurve=DataInterface() {propertyValues {Curve {
         defaultValue=1
         keys [
          {arriveTangent=-1.7885, interpMode=RCIM_Cubic, leaveTangent=-1.7885, tangentMode=RCTM_User, time=0.05, value=1}
          {interpMode=RCIM_Cubic, tangentMode=RCTM_User, time=1, value=0}
         ]
        }}})
    )
    ApplyForce=false
    ApplyPosition=true
    ApplyPositionFalloff=false
   ) {enabled=false, module=InheritSourceMovement001}
   &SubUVAnimation(
    Evaluation Type=::NewEnumerator0
    "Sprite Renderer "=DataInterface() {propertyValues {SpriteRenderer=&self:Stars_0.NiagaraSpriteRendererProperties_3}}
    SubUV Animation Mode:ENiagara_SubUVLookupModeV2=::NewEnumerator2
    UseEndFrame=false
    UseStartFrame=false
   )
   &ScaleColor(
    Scale Alpha=&FloatFromCurve(FloatCurve=DataInterface() {propertyValues {Curve {
        defaultValue=0
        keys [
         {interpMode=RCIM_Cubic, tangentMode=RCTM_User, time=0, value=0}
         {interpMode=RCIM_Cubic, tangentMode=RCTM_User, time=0.5, value=1}
         {tangentMode=RCTM_User, time=1, value=0}
        ]
       }}})
    ScaleA=true
    ScaleRGB=true
   )
   &SolveForcesAndVelocity(Clamp Velocity=false, Limit Acceleration=false)
  ]
  rendererClasses [&NiagaraSpriteRendererProperties]
  renderers [{rendererClass=&NiagaraSpriteRendererProperties}]
 }
}
summary {
 systemName=NS_Boost_Sparkles
 emitters [
  {emitterName=Stars, rendererClasses [&NiagaraSpriteRendererProperties]}
  {emitterName=Ribbon_Default, rendererClasses [&NiagaraRibbonRendererProperties]}
  {emitterName=Ribbon_Color, rendererClasses [&NiagaraRibbonRendererProperties]}
 ]
 userVariables [
  User.bIsFrontendPreview {type=NiagaraBool, defaultValue=true}
  User.bIsPainted {type=NiagaraBool}
  User.ColorVariantID {type=NiagaraInt32}
  User.DTCarOffset {type=Vector3f, description="", defaultValue {struct=Vector3f, value=<0 0 0>}}
  User.GameModeID {type=NiagaraInt32}
  "User.OG = 4 vs Proposal = 15" {type=NiagaraFloat, defaultValue=14, description=""}
  User.PaintedColor {type=LinearColor, description="", defaultValue {struct=LinearColor, value=<rgba 1 1 1 1>}}
  User.PaintedScalar {type=NiagaraFloat, defaultValue=1, description=""}
  User.PortCount {type=NiagaraInt32, defaultValue=2}
 ]
}
